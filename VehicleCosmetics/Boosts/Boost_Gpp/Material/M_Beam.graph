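# Translucent, unlit, two-sided beam material for Niagara ribbons in the Gpp vehicle boost cosmetic.
# Panned, flow-distorted noise warps the V coordinate into a wavy band, which shapes a hue-shifted Core color with scrolling sine pulses.
# Dynamic parameters drive brightness, edge hardness, core pulses and an IsPainted switch to desaturated color.
# Particle alpha scales UV edge fades, and time-driven sine/cosine waves form a transformed wobble vector that fades at both ends.
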Material /VehicleCosmetics/Boosts/Boost_Gpp/Material/M_Beam
graph {
 properties {
  bUsedWithNiagaraRibbons=true
  blendMode=BLEND_Translucent
  pixelDepthOffsetMode=PDOM_Legacy
  refractionMethod=RM_PixelNormalOffset
  shadingModel=MSM_Unlit
  twoSided=true
  usedShadingModels=Unlit
 }
 nodes [
  n1 = Constant() {r=0.5}
  n2 = TextureCoordinate() {uTiling=2, vTiling=0.125}
  n3 = Constant() {r=0.5}
  n4 = TextureSample(UVs=Panner(Coordinate=TextureCoordinate() {uTiling=0.25, vTiling=0.25}) {speedX=-0.125, speedY=0.01}) {texture=/DealershipSharedLibrary/FX/Textures/Noise/T_VCVFX_Flow02_N}
  n5 = Multiply(A=Constant() {r=0.05}, B=AppendVector(A=@n4.R, B=@n4.R))
  n6 = Panner(Coordinate=Add(A=Panner(Coordinate=@n2, Time=@n3) {speedY=0.5}, B=@n5)) {speedX=-1}
  n7 = TextureCoordinate() {uTiling=1.75, vTiling=-0.125}
  n8 = Panner(Coordinate=Add(A=Panner(Coordinate=@n7, Time=@n3) {speedY=-0.5}, B=@n5)) {speedX=-1}
  n9 = ComponentMask(None=Add(
    A=TextureSample(UVs=@n6) {texture=/DealershipSharedLibrary/FX/Textures/Noise/T_VCVFX_DistortNoise02_RGB}.RGB
    B=TextureSample(UVs=@n8) {texture=&T_VCVFX_DistortNoise02_RGB}.RGB
   )) {g=true, r=true}
  n10 = ComponentMask(None=Add(A=TextureCoordinate(), B=Multiply(A=@n1, B=@n9))) {g=true}
  MaterialExpressionConstantClamp_9 = Saturate(None=ConstantBiasScale(None=@n10) {bias=-0.15, scale=1})
  MaterialExpressionConstantClamp_10 = Saturate(None=ConstantBiasScale(None=OneMinus(None=@n10)) {bias=-0.01, scale=1})
  n11 = Multiply(A=Constant() {r=6}, B=Multiply(A=@MaterialExpressionConstantClamp_9, B=@MaterialExpressionConstantClamp_10))
  n12 = Add(A=Saturate(None=@n11), B=Constant() {r=0.2})
  n13 = Constant() {r=0.3}
  MaterialExpressionConstantClamp_8 = Saturate(None=@n11)
  MaterialExpressionConstantClamp_5 = Saturate(None=Multiply(A=@MaterialExpressionConstantClamp_8, B=@MaterialExpressionConstantClamp_8))
  n14 = Add(A=@MaterialExpressionConstantClamp_5, B=Constant() {r=0.2})
  n15 = Constant() {r=0.2}
  AppendVector(A=AppendVector(A=Multiply(A=@n12, B=@n13), B=Multiply(A=@n14, B=@n15)), B=Add(A=@MaterialExpressionConstantClamp_8, B=Constant() {r=0.2}))
  Constant3Vector() {constant=<rgba 1 1 1 0>}
  MaterialExpressionVertexColor_0 = ParticleColor()
  n16 = DynamicParameter() {paramNames [Brightness,Hardness,Core,Param4]}
  n17 = ComponentMask(None=TextureCoordinate()) {g=true}
  n18 = OneMinus(None=@n17)
  MaterialExpressionConstantClamp_3 = Saturate(None=ConstantBiasScale(None=Multiply(A=@n17, B=@n18)) {bias=-0.05, scale=6})
  MaterialExpressionConstantClamp_2 = Saturate(None=Multiply(A=@MaterialExpressionConstantClamp_3, B=@MaterialExpressionConstantClamp_3))
  n19 = ComponentMask(None=TextureCoordinate()) {r=true}
  MaterialExpressionConstantClamp_4 = Saturate(None=ConstantBiasScale(None=@n19) {bias=-0.025, scale=6})
  n20 = Multiply(A=OneMinus(None=@n19), B=@MaterialExpressionConstantClamp_4)
  MaterialExpressionConstantClamp_1 = Saturate(None=Multiply(A=@n16.Hardness, B=Multiply(A=@MaterialExpressionConstantClamp_2, B=@n20)))
  MaterialExpressionConstantClamp_0 = Saturate(None=Multiply(A=@MaterialExpressionVertexColor_0.A, B=@MaterialExpressionConstantClamp_1))
  n21 = VectorParameter() {parameterName=Core, defaultValue=<rgba 0.9723 0.1123 0 1>}
  n22 = Desaturation(None=@n21.RGB)
  n23 = If(
   A=Saturate(None=DynamicParameter() {parameterIndex=3, defaultValue=<rgba 0 1 1 1>, paramNames [IsPainted?, "", "", ""]}.IsPainted?)
   "A > B"=@n22
   "A == B"=@n22
   "A < B"=@n21.RGB
  ) {constB=0.5}
  n24 = Multiply(A=@n23, B=@n23)
  n25 = MaterialFunctionCall(Hue Shift Percentage=ScalarParameter() {defaultValue=0.15, parameterName=HueShift}, Texture=@n23) {materialFunction=&HueShift}
  n26 = Multiply(A=Desaturation(None=@n25.Result, Fraction=Constant() {r=0.1}), B=Constant() {r=2})
  n27 = ComponentMask(None=TextureCoordinate()) {g=true}
  n28 = OneMinus(None=@n27)
  MaterialExpressionConstantClamp_6 = Saturate(None=ConstantBiasScale(None=Multiply(A=@n27, B=@n28)) {bias=-0.1, scale=5})
  n29 = Multiply(A=@MaterialExpressionConstantClamp_6, B=@MaterialExpressionConstantClamp_6)
  n30 = Multiply(A=TextureCoordinate() {uTiling=0.25}, B=AppendVector(A=@n16.Param4, B=Constant() {r=1}))
  n31 = Add(A=ComponentMask(None=Frac(None=@n30)) {r=true}, B=OneMinus(None=Time()))
  MaterialExpressionConstantClamp_7 = Saturate(None=ConstantBiasScale(None=Sine(None=@n31) {period=0.5}) {bias=-0.1, scale=0.8})
  n32 = Multiply(A=Multiply(A=@MaterialExpressionConstantClamp_5, B=@MaterialExpressionConstantClamp_7), B=@n16.Core)
  n33 = Multiply(
   A=LinearInterpolate(A=LinearInterpolate(A=@n24, B=@n26, Alpha=@n29), B=SquareRoot(None=@n25.Result), Alpha=Multiply(A=@n32, B=@n32))
   B=@MaterialExpressionConstantClamp_5
  )
  MaterialFunctionCall(Emissive=Multiply(A=Multiply(A=@n33, B=@n16.Brightness), B=@MaterialExpressionVertexColor_0.RGB)) {materialFunction=&MF_FX_EmissiveAdjustments}
  n34 = Multiply(A=OneMinus(None=Time()), B=ScalarParameter() {defaultValue=1, parameterName=Speed})
  n35 = Divide(A=ComponentMask(None=TextureCoordinate()) {r=true}, B=ScalarParameter() {defaultValue=1, parameterName=tile})
  n36 = Add(A=@n34, B=@n35)
  n37 = Add(A=@n35, B=@n34)
  n38 = AppendVector(A=AppendVector(A=Cosine(None=@n36) {period=0.5}, B=Sine(None=@n37)), B=Constant())
  n39 = Sine(None=@n36) {period=0.75}
  n40 = Cosine(None=@n37)
  n41 = AppendVector(A=AppendVector(A=OneMinus(None=@n39), B=OneMinus(None=@n40)), B=Constant())
  MaterialExpressionConstantClamp_12 = Saturate(None=LinearInterpolate(A=Sine(None=@n37) {period=2}, B=Cosine(None=@n36) {period=2.5}, Alpha=Constant() {r=0.5}))
  n42 = Multiply(
   A=Multiply(
    A=LinearInterpolate(A=@n38, B=@n41, Alpha=@MaterialExpressionConstantClamp_12)
    B=ScalarParameter() {defaultValue=1, parameterName=Intensity}
   )
   B=ScalarParameter() {defaultValue=1, parameterName=Amount}
  )
  n43 = ComponentMask(None=TextureCoordinate()) {r=true}
  n44 = OneMinus(None=@n43)
  MaterialExpressionConstantClamp_11 = Saturate(None=Multiply(A=Multiply(A=@n44, B=@n43), B=Constant() {r=4}))
  Multiply(A=ConstantBiasScale(None=Transform(None=@n42)), B=@MaterialExpressionConstantClamp_11)
 ]
}
