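# Translucent, unlit, two-sided Niagara mesh-particle material that draws flickering lightning bolts for the Gpp vehicle boost.
# Scrolling sine and cosine bands mask channel-packed bolt textures and an optional flipbook, with particle color G and B weighting layers.
# Cloud noise panned over zero-tiled UVs flickers whole meshes between layers and brightness spikes.
# Dynamic parameters drive band motion, brightness, hardness, and an IsPainted switch between white and CustomColor.
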Material /VehicleCosmetics/Boosts/Boost_Gpp/Material/M_BoostMesh_Base_02
graph {
 properties {
  bUsedWithNiagaraMeshParticles=true
  blendMode=BLEND_Translucent
  opacityMaskClipValue=0.8
  pixelDepthOffsetMode=PDOM_Legacy
  shadingModel=MSM_Unlit
  twoSided=true
  usedShadingModels=Unlit
 }
 nodes [
  Constant4Vector() {constant=<rgba 1 1 1 1>}
  MaterialExpressionMeshEmitterDynamicParameter_2 = DynamicParameter() {paramNames [Brithness,PanSpeed,Tile,Hardness]}
  n1 = Multiply(A=TextureCoordinate(), B=AppendVector(A=Constant() {r=1}, B=@MaterialExpressionMeshEmitterDynamicParameter_2.Tile))
  n2 = Add(A=ComponentMask(None=Frac(None=@n1)) {g=true}, B=Multiply(A=Time(), B=@MaterialExpressionMeshEmitterDynamicParameter_2.PanSpeed))
  n3 = Cosine(None=@n2)
  MaterialExpressionConstantClamp_11 = Saturate(None=ConstantBiasScale(None=@n3) {bias=-0.1, scale=1})
  n4 = Multiply(A=ComponentMask(None=TextureCoordinate() {vTiling=0}) {r=true}, B=ScalarParameter() {defaultValue=1, parameterName=02_PatternTileU})
  n5 = ScalarParameter() {parameterName=02_Pattern_PanU}
  n6 = Multiply(A=ComponentMask(None=TextureCoordinate() {uTiling=0}) {g=true}, B=ScalarParameter() {defaultValue=1, parameterName=02_PatternTileV})
  n7 = ScalarParameter() {parameterName=02_Pattern_PanV}
  n8 = AppendVector(
   A=ComponentMask(None=Panner(Coordinate=@n4, Time=@n5) {speedX=0.5}) {r=true}
   B=ComponentMask(None=Panner(Coordinate=@n6, Time=@n7) {speedY=0.5}) {g=true}
  )
  n9 = Constant() {r=1}
  n10 = ScalarParameter() {defaultValue=4, parameterName=02_Rows}
  n11 = Divide(A=@n9, B=@n10)
  n12 = ScalarParameter() {defaultValue=4, parameterName=02_Columns}
  n13 = Divide(A=@n9, B=@n12)
  n14 = Multiply(A=Time(), B=ScalarParameter() {defaultValue=12, parameterName=02_FB_Speed})
  n15 = Fmod(A=Floor(None=@n14), B=@n10)
  n16 = Floor(None=Divide(A=@n14, B=@n10))
  n17 = Add(A=Multiply(A=@n8, B=AppendVector(A=@n11, B=@n13)), B=AppendVector(A=Divide(A=@n15, B=@n10), B=Divide(A=@n16, B=@n12)))
  n18 = Multiply(
   A=@MaterialExpressionConstantClamp_11
   B=TextureSampleParameter2D(UVs=StaticSwitchParameter(True=@n17, False=@n8) {defaultValue=true, parameterName=FlipBook}) {parameterName=LightningTexture, texture=/VehicleCosmetics/Boosts/Boost_Gpp/Textures/T_VCVFX_Boost_Gpp_BoltBaseFB_RGB}.R
  )
  MaterialExpressionMeshEmitterVertexColor_2 = ParticleColor()
  n19 = Constant()
  n20 = Sine(None=@n2) {period=0.5}
  MaterialExpressionConstantClamp_13 = Saturate(None=ConstantBiasScale(None=@n20) {bias=-0.05, scale=1})
  n21 = TextureSample() {texture=/VehicleCosmetics/Boosts/Boost_Gpp/Textures/T_VCVFX_Boost_Gpp_BoltBase02_RGB}
  n22 = Multiply(A=@MaterialExpressionConstantClamp_13, B=@n21.R)
  n23 = Multiply(A=@MaterialExpressionConstantClamp_11, B=@n21.B)
  n24 = TextureCoordinate() {uTiling=0, vTiling=0}
  n25 = Multiply(A=Time(), B=ScalarParameter() {defaultValue=0.25, parameterName=PanSpeed})
  n26 = Panner(Coordinate=@n24, Time=@n25) {speedX=1, speedY=0.65}
  n27 = Panner(Coordinate=@n24, Time=@n25) {speedX=-1, speedY=-0.3}
  n28 = AppendVector(
   A=AppendVector(
    A=TextureSample(UVs=@n26) {texture=/VehicleCosmetics/FX/Textures/Noise/T_VCVFX_Clouds_A}.R
    B=TextureSample(UVs=@n27) {texture=&T_VCVFX_Clouds_A}.R
   )
   B=TextureSample(UVs=Panner(Coordinate=@n24, Time=@n25) {speedX=-0.5, speedY=0.3}) {texture=&T_VCVFX_Clouds_A}.R
  )
  n29 = Constant() {r=0.2}
  MaterialExpressionConstantClamp_14 = Saturate(None=Multiply(A=Subtract(A=@n28, B=@n29), B=Constant() {r=35}))
  n30 = ComponentMask(None=@MaterialExpressionConstantClamp_14) {g=true}
  n31 = LinearInterpolate(A=@n23, B=Multiply(A=@n23, B=Constant() {r=8}), Alpha=@n30)
  MaterialExpressionConstantClamp_12 = Saturate(None=ConstantBiasScale(None=Multiply(A=@n3, B=@n20)) {bias=-0.1, scale=1})
  n32 = LinearInterpolate(
   A=LinearInterpolate(A=Constant(), B=Add(A=@n22, B=@n31), Alpha=@MaterialExpressionMeshEmitterVertexColor_2.B)
   B=Multiply(A=@MaterialExpressionConstantClamp_12, B=@n21.G)
   Alpha=@MaterialExpressionMeshEmitterVertexColor_2.G
  )
  n33 = ComponentMask(None=@MaterialExpressionConstantClamp_14) {r=true}
  n34 = Constant()
  n35 = TextureSample() {texture=/VehicleCosmetics/Boosts/Boost_Gpp/Textures/T_VCVFX_Boost_Gpp_BoltBase_RGB}
  n36 = Multiply(A=@MaterialExpressionConstantClamp_13, B=@n35.B)
  n37 = Constant() {r=8}
  n38 = ComponentMask(None=@MaterialExpressionConstantClamp_14) {b=true}
  n39 = Add(A=Multiply(A=@MaterialExpressionConstantClamp_12, B=@n35.R), B=LinearInterpolate(A=@n36, B=Multiply(A=@n36, B=@n37), Alpha=@n38))
  n40 = Multiply(
   A=LinearInterpolate(
    A=LinearInterpolate(A=@n34, B=@n39, Alpha=@MaterialExpressionConstantClamp_11)
    B=Multiply(A=@MaterialExpressionConstantClamp_11, B=@n35.G)
    Alpha=@n33
   )
   B=@MaterialExpressionMeshEmitterVertexColor_2.G
  )
  n41 = Constant() {r=5}
  n42 = LinearInterpolate(
   A=Add(A=Multiply(A=@n18, B=@MaterialExpressionMeshEmitterVertexColor_2.B), B=LinearInterpolate(A=@n19, B=@n32, Alpha=@n33))
   B=LinearInterpolate(A=@n40, B=Multiply(A=@n41, B=@n40), Alpha=@n38)
   Alpha=@n30
  )
  MaterialExpressionConstantClamp_10 = Saturate(None=Multiply(A=@MaterialExpressionMeshEmitterDynamicParameter_2.Hardness, B=@n42))
  n43 = Constant3Vector() {constant=<rgba 1 1 1 0>}
  n44 = If(
   A=Saturate(None=DynamicParameter() {parameterIndex=3, paramNames [IsPainted?, "", "", ""]}.IsPainted?)
   "A > B"=@n43
   "A == B"=@n43
   "A < B"=VectorParameter() {parameterName=CustomColor, defaultValue=<rgba 0.3 0.25 1 1>}.RGB
  ) {constB=0.5}
  MaterialFunctionCall(Emissive=Multiply(A=Multiply(A=@n44, B=@n42), B=@MaterialExpressionMeshEmitterDynamicParameter_2.Brithness)) {materialFunction=&MF_FX_EmissiveAdjustments}
 ]
}
