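# Translucent, unlit, two-sided Niagara ribbon material that draws flickering lightning bolts for the Gpp vehicle boost cosmetic.
# Two animated flipbook textures have their channels blended by panning noise, sine waves and a scrolling trail mask, with Dynamic Parameters driving brightness, flicker and pan speed.
# Noise masking, U-axis fades and particle color shape emissive and depth-faded opacity.
# A separate sine/cosine branch builds an animated wobble vector, tapered at both ribbon ends.
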Material /VehicleCosmetics/Boosts/Boost_Gpp/Material/M_LightningBolts
graph {
 properties {
  bUsedWithNiagaraRibbons=true
  blendMode=BLEND_Translucent
  pixelDepthOffsetMode=PDOM_Legacy
  shadingModel=MSM_Unlit
  twoSided=true
  usedShadingModels=Unlit
 }
 nodes [
  MaterialExpressionVertexColor_2 = ParticleColor()
  n1 = TextureCoordinate() {uTiling=0.5, vTiling=0.1}
  n2 = Multiply(A=ScalarParameter() {defaultValue=2, parameterName=MaskSpeed}, B=Time())
  n3 = TextureSample(UVs=Panner(Coordinate=@n1, Time=@n2) {speedX=-1, speedY=0.65}) {texture=/DealershipSharedLibrary/FX/Textures/Noise/T_VCVFX_NoiseCloud_RGB}
  n4 = TextureSample(UVs=Panner(Coordinate=@n1, Time=@n2) {speedX=-0.75, speedY=-1}) {texture=&T_VCVFX_NoiseCloud_RGB}
  n5 = AppendVector(A=Add(A=@n3.B, B=@n4.B), B=Add(A=@n3.R, B=@n4.R))
  n6 = Constant()
  MaterialExpressionConstantClamp_21 = Saturate(None=ConstantBiasScale(None=AppendVector(A=@n5, B=@n6)) {bias=-0.1, scale=1})
  n7 = StaticSwitchParameter(
   True=ComponentMask(None=@MaterialExpressionConstantClamp_21) {r=true}
   False=ComponentMask(None=@MaterialExpressionConstantClamp_21) {g=true}
  ) {parameterName=Mask}
  n8 = DynamicParameter() {paramNames [Brightness,FlickerSpeed,PanSpeed,Param4]}
  n9 = TextureCoordinate() {vTiling=0}
  n10 = Panner(
   Coordinate=Multiply(A=ComponentMask(None=@n9) {r=true}, B=ScalarParameter() {defaultValue=1, parameterName=PatternTileU})
   Time=ScalarParameter() {parameterName=Pattern_PanU}
  ) {speedY=0.5}
  n11 = TextureCoordinate() {uTiling=0}
  n12 = Panner(
   Coordinate=Multiply(A=ComponentMask(None=@n11) {g=true}, B=ScalarParameter() {defaultValue=1, parameterName=PatternTileV})
   Time=ScalarParameter() {parameterName=Pattern_PanV}
  ) {speedY=0.5}
  n13 = Constant() {r=1}
  n14 = ScalarParameter() {defaultValue=2, parameterName=Columns}
  n15 = ScalarParameter() {defaultValue=4, parameterName=Rows}
  n16 = Multiply(
   A=AppendVector(A=ComponentMask(None=@n10) {r=true}, B=ComponentMask(None=@n12) {g=true})
   B=AppendVector(A=Divide(A=@n13, B=@n14), B=Divide(A=@n13, B=@n15))
  )
  n17 = Multiply(A=Time(), B=ScalarParameter() {defaultValue=16, parameterName=FB_Speed})
  n18 = Floor(None=@n17)
  n19 = Divide(A=@n17, B=@n15)
  n20 = AppendVector(A=Divide(A=Fmod(A=@n18, B=@n15), B=@n15), B=Divide(A=Floor(None=@n19), B=@n14))
  n21 = Multiply(A=@n8.PanSpeed, B=Time())
  n22 = TextureSampleParameter2D(UVs=Panner(Coordinate=Add(A=@n16, B=@n20), Time=@n21) {speedX=-1}) {parameterName=Texture01, texture=/DealershipSharedLibrary/FX/Textures/TexturePackage/T_VCVFX_LightningFB01_RGB}
  n23 = Multiply(A=@n8.Param4, B=@n22.B)
  n24 = Multiply(A=@n8.Brightness, B=@n22.R)
  n25 = TextureCoordinate() {uTiling=0.1, vTiling=0}
  n26 = Multiply(A=@n8.FlickerSpeed, B=Time())
  n27 = TextureSample(UVs=Panner(Coordinate=@n25, Time=@n26) {speedX=1, speedY=0.65}) {texture=&T_VCVFX_NoiseCloud_RGB}
  n28 = TextureSample(UVs=Panner(Coordinate=@n25, Time=@n26) {speedX=0.3, speedY=-1}) {texture=&T_VCVFX_NoiseCloud_RGB}
  n29 = Add(A=@n27.B, B=@n28.B)
  n30 = Add(A=@n27.R, B=@n28.R)
  MaterialExpressionConstantClamp_24 = Saturate(None=AppendVector(A=AppendVector(A=@n29, B=@n30), B=Add(A=@n27.G, B=@n28.G)))
  n31 = ComponentMask(None=@MaterialExpressionConstantClamp_24) {b=true}
  n32 = ComponentMask(None=@MaterialExpressionConstantClamp_24) {g=true}
  n33 = LinearInterpolate(A=LinearInterpolate(A=@n23, B=@n24, Alpha=@n31), B=@n22.G, Alpha=@n32)
  n34 = ComponentMask(None=@MaterialExpressionConstantClamp_24) {r=true}
  n35 = LinearInterpolate(A=LinearInterpolate(A=@n24, B=@n22.G, Alpha=@n32), B=@n23, Alpha=@n34)
  n36 = Add(A=Multiply(A=Constant() {r=0.2}, B=@n33), B=Multiply(A=Constant() {r=3}, B=@n35))
  n37 = Multiply(A=TextureCoordinate(), B=AppendVector(A=Constant() {r=1}, B=Constant() {r=1}))
  n38 = Time()
  n39 = Add(A=ComponentMask(None=Frac(None=@n37)) {r=true}, B=Multiply(A=OneMinus(None=@n38), B=Constant() {r=2}))
  n40 = Cosine(None=@n39)
  n41 = Sine(None=@n39) {period=0.5}
  MaterialExpressionConstantClamp_23 = Saturate(None=ConstantBiasScale(None=Add(A=@n40, B=@n41)) {bias=-0.1, scale=1})
  n42 = Add(A=Multiply(A=Multiply(A=@MaterialExpressionConstantClamp_23, B=@n33), B=Constant() {r=8}), B=Multiply(A=@n35, B=Constant() {r=0.2}))
  MaterialExpressionConstantClamp_22 = Saturate(None=Add(A=@n34, B=@n31))
  n43 = Abs(None=Add(A=TextureCoordinate() {uTiling=0.2, vTiling=2}, B=Constant() {r=-1}))
  n44 = Add(A=Multiply(A=Constant2Vector() {g=1, r=16}, B=OneMinus(None=@n43)), B=Constant2Vector() {g=0.15, r=1.3})
  n45 = Constant() {r=-0.06}
  n46 = ScalarParameter() {defaultValue=2, parameterName=MaskSpeed}
  n47 = Time()
  n48 = TextureSample(UVs=Panner(Coordinate=Rotator(Coordinate=@n44, Time=@n45) {speed=3.14}, Time=Multiply(A=@n46, B=@n47)) {speedX=-1, speedY=2}) {texture=/DealershipSharedLibrary/FX/Textures/Noise/T_VCVFX_TrailMask_RGB}
  n49 = TextureCoordinate() {vTiling=0}
  n50 = Panner(
   Coordinate=Multiply(A=ComponentMask(None=@n49) {r=true}, B=ScalarParameter() {defaultValue=1, parameterName=A_PatternTileU})
   Time=ScalarParameter() {parameterName=A_Pattern_PanU}
  ) {speedX=0.5}
  n51 = TextureCoordinate() {uTiling=0}
  n52 = Panner(
   Coordinate=Multiply(A=ComponentMask(None=@n51) {g=true}, B=ScalarParameter() {defaultValue=1, parameterName=A_PatternTileV})
   Time=ScalarParameter() {parameterName=A_Pattern_PanV}
  ) {speedY=0.5}
  n53 = Constant() {r=1}
  n54 = ScalarParameter() {defaultValue=2, parameterName=A_Columns}
  n55 = ScalarParameter() {defaultValue=4, parameterName=A_Rows}
  n56 = Multiply(
   A=AppendVector(A=ComponentMask(None=@n50) {r=true}, B=ComponentMask(None=@n52) {g=true})
   B=AppendVector(A=Divide(A=@n53, B=@n54), B=Divide(A=@n53, B=@n55))
  )
  n57 = Multiply(A=Time(), B=ScalarParameter() {defaultValue=16, parameterName=A_FB_Speed})
  n58 = Floor(None=@n57)
  n59 = Divide(A=@n57, B=@n55)
  n60 = AppendVector(A=Divide(A=Fmod(A=@n58, B=@n55), B=@n55), B=Divide(A=Floor(None=@n59), B=@n54))
  n61 = TextureSampleParameter2D(UVs=Panner(Coordinate=Add(A=@n56, B=@n60), Time=@n21) {speedX=-1.25}) {parameterName=ATexture, texture=/VehicleCosmetics/Boosts/Boost_Gpp/Textures/T_VCVFX_Boost_Gpp_GPPLightningFB04_RGB}
  n62 = LinearInterpolate(A=Multiply(A=@n48.R, B=@n61.R), B=Multiply(A=@n48.G, B=@n61.G), Alpha=@n32)
  n63 = Multiply(A=@n48.B, B=@n61.B)
  n64 = Add(
   A=Multiply(A=@n8.Brightness, B=LinearInterpolate(A=@n36, B=@n42, Alpha=@MaterialExpressionConstantClamp_22))
   B=Multiply(A=@n8.Param4, B=LinearInterpolate(A=@n62, B=@n63, Alpha=@n34))
  )
  n65 = ComponentMask(None=TextureCoordinate()) {r=true}
  n66 = OneMinus(None=@n65)
  MaterialExpressionConstantClamp_25 = Saturate(None=ConstantBiasScale(None=@n65) {bias=-0.025, scale=4})
  n67 = Multiply(A=@n7, B=Multiply(A=Multiply(A=@n64, B=@n66), B=@MaterialExpressionConstantClamp_25))
  MaterialExpressionConstantClamp_20 = Saturate(None=Multiply(A=Multiply(A=@MaterialExpressionVertexColor_2.A, B=@n67), B=DepthFade()))
  n68 = TextureCoordinate()
  n69 = Multiply(A=OneMinus(None=ComponentMask(None=@n68) {r=true}), B=Constant() {r=8})
  n70 = Multiply(A=@n7, B=@n64)
  MaterialFunctionCall(Emissive=Multiply(A=@MaterialExpressionVertexColor_2.RGB, B=Multiply(A=@n69, B=@n70))) {materialFunction=&MF_FX_EmissiveAdjustments}
  n71 = Time()
  n72 = TextureCoordinate()
  n73 = Add(
   A=Multiply(A=OneMinus(None=@n71), B=ScalarParameter() {defaultValue=1, parameterName=Speed})
   B=Divide(A=ComponentMask(None=@n72) {r=true}, B=ScalarParameter() {defaultValue=1, parameterName=tile})
  )
  n74 = Cosine(None=@n73) {period=0.75}
  n75 = Sine(None=@n73) {period=0.75}
  n76 = Sine(None=@n73)
  n77 = AppendVector(A=AppendVector(A=@n74, B=@n75), B=Multiply(A=@n74, B=@n76))
  n78 = Cosine(None=@n73)
  n79 = AppendVector(A=AppendVector(A=@n76, B=@n78), B=Multiply(A=@n75, B=@n78))
  n80 = Multiply(A=@n71, B=ScalarParameter() {defaultValue=1, parameterName=BlendTile})
  MaterialExpressionConstantClamp_28 = Saturate(None=LinearInterpolate(A=Sine(None=@n80) {period=5}, B=Cosine(None=@n80) {period=5}, Alpha=Constant() {r=0.5}))
  n81 = Multiply(
   A=Multiply(
    A=LinearInterpolate(A=@n77, B=@n79, Alpha=@MaterialExpressionConstantClamp_28)
    B=ScalarParameter() {defaultValue=1, parameterName=Intensity}
   )
   B=ScalarParameter() {defaultValue=1, parameterName=Amount}
  )
  n82 = ComponentMask(None=TextureCoordinate()) {r=true}
  n83 = OneMinus(None=@n82)
  MaterialExpressionConstantClamp_27 = Saturate(None=Multiply(A=Multiply(A=@n83, B=@n82), B=Constant() {r=4}))
  Multiply(A=ConstantBiasScale(None=Transform(None=@n81)) {bias=-0.05, scale=0.7}, B=@MaterialExpressionConstantClamp_27)
 ]
}
